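# Physics asset for the Miner pickaxe, a two-handed melee weapon cosmetic, giving its skeletal mesh a handle body and a dangling strap body.
# A single large box covers the Melee_TwoHanded_Handle bone, and a small capsule covers the dyn_strap_01 bone.
# One constraint joins strap to handle with twist and Swing1 locked, so the strap rotates only on the Swing2 axis, like a hinge.
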PhysicsAsset /BRCosmetics/Weapons/FORT_Melee/Pickaxe_Miner/Meshes/Miner_Axe_PhysicsAsset
bodies {
 Melee_TwoHanded_Handle [{shapeType=Box, extentX=71.692, extentY=37.5191, extentZ=114.8859, center=<0.7486 2.0432 29.0933>, rotation=<rot 0.4136 -90 0.014>}]
 dyn_strap_01 [{shapeType=Capsule, length=6.7088, radius=5.8611, center=<5.2237 0 0.0542>, rotation=<rot 0 90 -89.3451>}]
}
constraints [{
  bone1Name=dyn_strap_01
  bone2Name=Melee_TwoHanded_Handle
  swing1LimitDegrees=45
  swing1Motion=Locked
  swing2LimitDegrees=90
  twistLimitDegrees=45
  twistMotion=Locked
 }]
